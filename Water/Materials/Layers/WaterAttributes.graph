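# Water layer material function that outputs a material attributes set with the base shading values for water surfaces.
# Base color is the albedo parameter divided by CausticsMaxIntensity from MPC_Water, and opacity is a constant zero.
# A Fresnel term blends specular and roughness between base and Fresnel parameters.
# Refraction lerps from near to far values using saturated pixel depth over a distance parameter.
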MaterialFunction /Water/Materials/Layers/WaterAttributes
graph {nodes [
  n1 = Fresnel(ExponentIn=ScalarParameter() {defaultValue=5, group=Water Shading, parameterName="Water Fresnel  Exponenth"}) {baseReflectFraction=0}
  n2 = Divide(A=PixelDepth().R, B=ScalarParameter() {defaultValue=5000, group=Water Shading, parameterName=Depth for Min Refraction})
  n3 = SetMaterialAttributes!(
   MaterialAttributes
   BaseColor=Divide(
    A=VectorParameter() {group=Water Shading, parameterName=Water Albedo, defaultValue=<rgba 0.85 0.85 0.85 0.5>}.RGB
    B=CollectionParameter() {collection=/Water/Materials/MPC/MPC_Water, parameterName=CausticsMaxIntensity}
   )
   Specular=LinearInterpolate(
    A=ScalarParameter() {defaultValue=0.225, group=Water Shading, parameterName=Water Specular}
    B=ScalarParameter() {defaultValue=0.225, group=Water Shading, parameterName=Water Fresnel Specular}
    Alpha=@n1
   )
   Roughness=LinearInterpolate(
    A=ScalarParameter() {defaultValue=0.02, group=Water Shading, parameterName=Water Roughness}
    B=ScalarParameter() {defaultValue=0.1, group=Water Shading, parameterName=Water Fresnel Roughness}
    Alpha=@n1
   )
   Opacity=Constant()
   Refraction=LinearInterpolate(
    A=ScalarParameter() {defaultValue=1.15, group=Water Shading, parameterName=Refraction, sliderMax=2}
    B=ScalarParameter() {defaultValue=1.02, group=Water Shading, parameterName=Refraction Far, sliderMax=2}
    Alpha=Saturate(None=@n2)
   )
  ) {attributeSetTypes [
    #1
    #2
    #3
    #4
    #5
   ]}
  FunctionOutput(None=@n3)
 ]}
